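; Behavior tree service for the CoffeePuff boss that drains the controlled pawn's energy while its melee combo branch is active.
; On activation it finds the pawn's Fort energy component and starts continuous energy use at a scalable per-second rate, tagged with the boss's MeleeCombo ability tag.
; On deactivation it stops energy use under the same tag.
; Both events validity-check the pawn and component first.
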(event AI|EventReceiveActivationAI
  (params (OwnerController :type AIController%20Object%20Reference) (ControlledPawn :type Pawn%20Object%20Reference))
  (bind _returnvalue (Actor|GetComponentbyClass ControlledPawn "/Script/EnergyRuntime.FortComponent_Energy"))
  (Utilities|IsValid ControlledPawn
    (:Is%20Valid
      (Utilities|IsValid _returnvalue
        (:Is%20Valid
          (Energy|StartUsingEnergy _returnvalue 0.0 (ScalableFloat|GetValueatLevel (Variables|Default|GetEnergyUsePerSecond)) (GameplayTags|MakeLiteralGameplayTag "(TagName=\"Ability.VividRazor.Boss.CoffeePuff.MeleeCombo\")") 0))))))

(event AI|EventReceiveDeactivationAI
  (params (OwnerController :type AIController%20Object%20Reference) (ControlledPawn :type Pawn%20Object%20Reference))
  (bind _returnvalue (Actor|GetComponentbyClass ControlledPawn "/Script/EnergyRuntime.FortComponent_Energy"))
  (Utilities|IsValid ControlledPawn
    (:Is%20Valid
      (Utilities|IsValid _returnvalue
        (:Is%20Valid
          (Energy|StopUsingEnergy _returnvalue (GameplayTags|MakeLiteralGameplayTag "(TagName=\"Ability.VividRazor.Boss.CoffeePuff.MeleeCombo\")") 0))))))
