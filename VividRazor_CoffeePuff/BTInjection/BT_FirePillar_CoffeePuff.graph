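# Behavior tree for the CoffeePuff AI that repositions it and fires an ability, apparently its Fire Pillar attack.
# A single Sequence runs an EQS reposition query, writes the single result to NextDestination and moves there, allowing partial paths.
# It then rotates to face the evaluator's target actor and executes the gameplay ability whose tags come from the Ability_TagContainer blackboard key.
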BehaviorTree /VividRazor_CoffeePuff/BTInjection/BT_FirePillar_CoffeePuff
blackboard=&/BB_CoffeePuff
root {
 name=Sequence
 class=BTComposite_Sequence
 children [
  Run EQS Query {
   class=BTTask_RunEQSQuery
   properties {
    nodeName=Run EQS Query
    blackboardKey {selectedKeyName=NextDestination}
    eQSRequest {queryTemplate=/VividRazor_CoffeePuff/Abilities/EQS/EQS_FirePillarReposition_CoffeePuff, runMode=SingleResult}
   }
  }
  Move To {
   class=BTTask_MoveTo
   children []
   decorators []
   properties {
    bIgnoreRestartSelf=false
    nodeName=Move To
    acceptableRadius {defaultValue=5, key=None}
    bAllowPartialPath {defaultValue=true, key=None}
    bAllowStrafe {defaultValue=false, key=None}
    bProjectGoalLocation {defaultValue=true, key=None}
    bReachTestIncludesAgentRadius {defaultValue=true, key=None}
    bReachTestIncludesGoalRadius {defaultValue=true, key=None}
    bRequireNavigableEndLocation {defaultValue=true, key=None}
    bStartFromPreviousPath {defaultValue=false, key=None}
    bTrackMovingGoal {defaultValue=true, key=None}
    blackboardKey {bNoneIsAllowedValue=false, selectedKeyName=NextDestination}
    filterClass {baseClass=&NavigationQueryFilter, defaultValue=null, key=None}
    observedBlackboardValueTolerance {defaultValue=4.75, key=None}
   }
   services []
  }
  Rotate to face BB entry {
   class=BTTask_RotateToFaceBBEntry
   decorators []
   properties {blackboardKey {selectedKeyName=AIEvaluator_Global_TargetActor}, precision {defaultValue=10}}
  }
  GameplayAbility {
   class=FortBTTask_ExecuteGameplayAbility
   decorators []
   properties {nodeName=GameplayAbility, gameplayAbilityTagContainer {key=Ability_TagContainer, defaultValue {gameplayTags []}}}
  }
 ]
 decorators []
 properties {children [
   {childTask=&self:BTTask_RunEQSQuery_0}
   {childTask=&self:BTTask_MoveTo_3}
   {childTask=&self:BTTask_RotateToFaceBBEntry_0}
   {childTask=&self:FortBTTask_ExecuteGameplayAbility_0}
  ]}
}
