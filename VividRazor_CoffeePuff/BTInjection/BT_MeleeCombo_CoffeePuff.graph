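# Behavior tree for CoffeePuff's melee combo, which closes on the target actor and then executes a gameplay ability.
# A Sequence runs a time-limited Move To while the target is beyond striking distance, aborting once in range or when the EnergyDepleted tag appears.
# Force Success on the move lets the sequence continue to the ability task, which takes its tags from the Ability_TagContainer blackboard key and is blocked if energy is depleted.
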BehaviorTree /VividRazor_CoffeePuff/BTInjection/BT_MeleeCombo_CoffeePuff
blackboard=&/BB_CoffeePuff
root {
 name=Sequence
 class=BTComposite_Sequence
 children [
  Move To {
   class=BTTask_MoveTo
   children []
   decorators [
    Distance Between {
     class=FortBTDecorator_DistanceBetween
     properties {
      flowAbortMode=Self
      nodeName=Distance Between
      operator=Greater
      blackboardKeyB {selectedKeyName=AIEvaluator_Global_TargetActor}
      specifiedDistance {defaultValue=320}
     }
    }
    Force Success {class=BTDecorator_ForceSuccess}
    TimeLimit {
     class=BTDecorator_TimeLimit
     properties {bInverseCondition=false, flowAbortMode=Self, nodeName=TimeLimit, timeLimit {defaultValue=3, key=None}}
    }
    Gameplay Tag Query {
     class=BTDecorator_GameplayTagQuery
     properties {
      flowAbortMode=Self
      nodeName=Gameplay Tag Query
      gameplayTagQuery {
       autoDescription=" NONE( Event.Ability.VividRazor.CoffeePuff.EnergyDepleted )"
       queryTokenStream [0,1,3,1,0]
       tagDictionary [{tagName=Event.Ability.VividRazor.CoffeePuff.EnergyDepleted}]
      }
     }
    }
   ]
   properties {
    bIgnoreRestartSelf=false
    nodeName=Move To
    acceptableRadius {defaultValue=5, key=None}
    bAllowPartialPath {defaultValue=true, key=None}
    bAllowStrafe {defaultValue=false, key=None}
    bProjectGoalLocation {defaultValue=true, key=None}
    bReachTestIncludesAgentRadius {defaultValue=true, key=None}
    bReachTestIncludesGoalRadius {defaultValue=true, key=None}
    bRequireNavigableEndLocation {defaultValue=true, key=None}
    bStartFromPreviousPath {defaultValue=false, key=None}
    bTrackMovingGoal {defaultValue=true, key=None}
    blackboardKey {bNoneIsAllowedValue=false, selectedKeyName=AIEvaluator_Global_TargetActor}
    filterClass {baseClass=&NavigationQueryFilter, defaultValue=null, key=None}
    observedBlackboardValueTolerance {defaultValue=4.75, key=None}
   }
   services []
  }
  GameplayAbility {
   class=FortBTTask_ExecuteGameplayAbility
   decorators [Gameplay Tag Condition {
     class=BTDecorator_CheckGameplayTagsOnActor
     properties {
      bInverseCondition=true
      flowAbortMode=None
      nodeName=Gameplay Tag Condition
      tagsToMatch=Any
      actorToCheck {bNoneIsAllowedValue=false, selectedKeyName=SelfActor}
      gameplayTags {gameplayTags [{tagName=Event.Ability.VividRazor.CoffeePuff.EnergyDepleted}]}
     }
    }]
   properties {nodeName=GameplayAbility, gameplayAbilityTagContainer {key=Ability_TagContainer, defaultValue {gameplayTags []}}}
  }
 ]
 decorators []
 properties {children [
   {
    childTask=&self:BTTask_MoveTo_0
    decorators [
     &self:FortBTDecorator_DistanceBetween_3
     &self:BTDecorator_ForceSuccess_3
     &self:BTDecorator_TimeLimit_3
     &self:BTDecorator_GameplayTagQuery_0
    ]
   }
   {
    childTask=&self:FortBTTask_ExecuteGameplayAbility_1
    decorators [&self:BTDecorator_CheckGameplayTagsOnActor_0]
   }
  ]}
}
