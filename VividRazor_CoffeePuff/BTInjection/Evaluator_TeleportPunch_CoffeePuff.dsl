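; Behavior tree evaluator for the CoffeePuff teleport punch, deciding whether the NPC's ongoing conditions for that action hold.
; EvaluateOngoingConditions casts the behavior tree component's owner to the NPC bot controller, then casts its controlled pawn to FortPawn.
; It returns true only while the pawn's movement component is not falling and the pawn lacks the Gameplay.Status.Knockback tag.
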(fn EvaluateOngoingConditions
  (params (OwnerComp :type Behavior%20Tree%20Component%20Object%20Reference))
  (bind _asbp_bot_controller_npc_base (Utilities|Casting|CastToBP_BotController_NPC_Base (Components|GetOwner OwnerComp)))
  (bind _asfort_pawn (Utilities|Casting|CastToFortPawn (Pawn|GetControlledPawn _asbp_bot_controller_npc_base)))
  (return (and (not (AI|Components|NavMovement|IsFalling (Variables|/Script/Engine.Character|GetCharacterMovement _asfort_pawn))) (not (GameplayTags|HasAnyMatchingGameplayTags _asfort_pawn (GameplayTags|MakeGameplayTagContainerfromArray (Utilities|Array|MakeArray "(TagName=\"Gameplay.Status.Knockback\")")))))))
